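# Material function for flipbook character faces that blends brows, eyes and mouth textures into a material's base color, emissive and opacity.
# Layers each feature's color by lerp behind static bool switches, and sums their alphas into a saturated face mask, optionally passing each through MF_Flipbook_DFLine for distance-field alpha.
# Static switch parameters choose which channels the face affects, with a strength parameter for emissive, and return three separate outputs.
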MaterialFunction /Game/Global/Materials/Character_Parent_2023/FlipbookFace_Parent_2023/MaterialFunctions/MF_Flipbook_Texture_Blend
graph {nodes [
  FunctionInput() {inputName=---------------------, inputType=FunctionInput_Scalar, sortPriority=-1}
  FunctionInput() {inputName=----------------------, inputType=FunctionInput_Scalar, sortPriority=4}
  FunctionInput() {inputName=------------------------, inputType=FunctionInput_Scalar, sortPriority=9}
  FunctionOutput(None=NamedRerouteUsage()) {outputName=Flipbook BaseColor}
  FunctionOutput(None=NamedRerouteUsage()) {outputName=Flipbook Emissive, sortPriority=1}
  FunctionOutput(None=NamedRerouteUsage()) {outputName=Flipbook Opacity, sortPriority=2}
  n1 = NamedRerouteUsage()
  n2 = Multiply(A=NamedRerouteUsage(), B=ScalarParameter() {defaultValue=1, group=Flipbook 01 Global, parameterName=FB_EmissiveStrength})
  n3 = NamedRerouteUsage()
  NamedRerouteDeclaration(None=StaticSwitchParameter(True=LinearInterpolate(A=@n1, B=@n2, Alpha=@n3), False=NamedRerouteUsage()) {group=Flipbook 01 Global, parameterName=FB_ApplyEmissive, sortPriority=3}) {name=Flipbook_Emissive, nodeColor=<rgba 0.108 0.2905 1 1>}
  n4 = GetMaterialAttributes(None=FunctionInput() {inputName=Material, inputType=FunctionInput_MaterialAttributes, sortPriority=-5}) {attributeGetTypes [#1, #2, #3]}
  NamedRerouteDeclaration(None=@n4.Opacity) {name=Opacity, nodeColor=<rgba 0 0.1328 1 1>}
  NamedRerouteDeclaration(None=FunctionInput() {inputName=Use Brows, inputType=FunctionInput_StaticBool, sortPriority=3}) {name=UseBrows, nodeColor=<rgba 0.1953 1 0 1>}
  NamedRerouteDeclaration(None=FunctionInput() {inputName=Use Eyes, inputType=FunctionInput_StaticBool, sortPriority=7}) {name=UseEyes, nodeColor=<rgba 0.1953 1 0 1>}
  NamedRerouteDeclaration(None=FunctionInput() {inputName=Use Mouth, inputType=FunctionInput_StaticBool, sortPriority=12}) {name=UseMouth, nodeColor=<rgba 0.1953 1 0 1>}
  NamedRerouteDeclaration(None=FunctionInput() {inputName=Brows Alpha, inputType=FunctionInput_Scalar, sortPriority=2}) {name=BrowsAlpha, nodeColor=<rgba 1 0 0.8438 1>}
  n5 = NamedRerouteUsage()
  n6 = NamedRerouteUsage()
  n7 = Add(
   A=StaticSwitch(
    True=MaterialFunctionCall!(DF=@n5) {materialFunction=/Game/Global/Materials/Character_Parent_2023/FlipbookFace_Parent_2023/MaterialFunctions/Common/MF_Flipbook_DFLine}.Result
    False=@n5
    Value=NamedRerouteUsage()
   )
   B=StaticSwitch(
    True=MaterialFunctionCall(DF=@n6) {materialFunction=&MF_Flipbook_DFLine}.Result
    False=@n6
    Value=NamedRerouteUsage()
   )
  )
  n8 = NamedRerouteUsage()
  n9 = StaticSwitch(
   True=MaterialFunctionCall(DF=@n8) {materialFunction=&MF_Flipbook_DFLine}.Result
   False=@n8
   Value=NamedRerouteUsage()
  )
  NamedRerouteDeclaration(None=Saturate(None=Add(A=@n7, B=@n9))) {name=FaceMask, nodeColor=<rgba 1 0.7969 0 1>}
  NamedRerouteDeclaration(None=FunctionInput() {inputName=Eyes Color, sortPriority=5}) {name=EyesColor, nodeColor=<rgba 0.2436 0 1 1>}
  NamedRerouteDeclaration(None=FunctionInput() {inputName=Eyes Alpha, inputType=FunctionInput_Scalar, sortPriority=6}) {name=EyesAlpha, nodeColor=<rgba 0.2436 0 1 1>}
  NamedRerouteDeclaration(None=FunctionInput() {inputName=Mouth Color, sortPriority=10}) {name=MouthColor, nodeColor=<rgba 0.2092 0.8266 1 1>}
  NamedRerouteDeclaration(None=FunctionInput() {inputName=Mouth Alpha, inputType=FunctionInput_Scalar, sortPriority=11}) {name=MouthAlpha, nodeColor=<rgba 0.2092 0.8266 1 1>}
  NamedRerouteDeclaration(None=FunctionInput() {inputName=Brows Color, sortPriority=1}) {name=BrowsColor, nodeColor=<rgba 1 0 0.8438 1>}
  n10 = StaticSwitch(True=NamedRerouteUsage(), False=Constant(), Value=NamedRerouteUsage())
  n11 = StaticSwitch(True=LinearInterpolate(A=@n10, B=NamedRerouteUsage(), Alpha=NamedRerouteUsage()), False=@n10, Value=NamedRerouteUsage())
  n12 = NamedRerouteUsage()
  n13 = NamedRerouteUsage()
  NamedRerouteDeclaration(None=StaticSwitch(True=LinearInterpolate(A=@n11, B=@n12, Alpha=@n13), False=@n11, Value=NamedRerouteUsage())) {name=FaceColor, nodeColor=<rgba 1 0.849 0.0986 1>}
  NamedRerouteDeclaration(None=StaticBoolParameter() {group=Flipbook 01 Global, parameterName=FB_UseDistanceFieldAlpha}) {name=Use Distance Field Alpha, nodeColor=<rgba 0 1 0.3203 1>}
  n14 = NamedRerouteUsage()
  n15 = NamedRerouteUsage()
  n16 = NamedRerouteUsage()
  NamedRerouteDeclaration(None=StaticSwitchParameter(True=LinearInterpolate(A=@n14, B=@n15, Alpha=@n16), False=NamedRerouteUsage()) {defaultValue=true, group=Flipbook 01 Global, parameterName=FB_ApplyBaseColor, sortPriority=2}) {name=Flipbook_BaseColor, nodeColor=<rgba 0.7578 1 0 1>}
  NamedRerouteDeclaration(None=StaticSwitchParameter(True=NamedRerouteUsage(), False=NamedRerouteUsage()) {group=Flipbook 01 Global, parameterName=FB_ApplyOpacityMask, sortPriority=4}) {name=Flipbook_Opacity, nodeColor=<rgba 0 1 0.6016 1>}
  NamedRerouteDeclaration(None=@n4.BaseColor) {name=BaseColor, nodeColor=<rgba 0 1 0.5078 1>}
  NamedRerouteDeclaration(None=@n4.EmissiveColor) {name=Emissive, nodeColor=<rgba 1 0.1875 0 1>}
 ]}
