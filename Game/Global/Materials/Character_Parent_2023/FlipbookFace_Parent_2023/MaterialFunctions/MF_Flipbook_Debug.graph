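# Debug material function for flipbook-face character materials, outputting two World Position Offsets and a vertex-color tint to visualize UVs and vertex data.
# The UV offset flattens the mesh into its UV0 or UV1 layout on a scaled XZ plane, with masked hiding.
# The vertex-color path quantizes a chosen channel into indices for three-column grid offsets, masks extreme values, and blends a debug color over Base Color.
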MaterialFunction /Game/Global/Materials/Character_Parent_2023/FlipbookFace_Parent_2023/MaterialFunctions/MF_Flipbook_Debug
graph {nodes [
  FunctionOutput(None=NamedRerouteUsage()) {outputName=Debug UV WPO}
  FunctionOutput(None=NamedRerouteUsage()) {outputName=Debug Vert Color, sortPriority=2}
  FunctionOutput(None=NamedRerouteUsage()) {outputName=Debug Vert Color WPO, sortPriority=1}
  NamedRerouteDeclaration(None=Subtract(A=ObjectPositionWS(), B=WorldPosition().XYZ)) {name=VertexColorHide, nodeColor=<rgba 0.1347 0.1879 1 1>}
  NamedRerouteDeclaration(None=VertexColor()) {name=VertexColor, nodeColor=<rgba 1 0.7274 0.0448 1>}
  NamedRerouteDeclaration(None=FunctionInput() {inputName=Vertex Color Hide Mask, inputType=FunctionInput_Scalar}) {name=VertexColorHideMask, nodeColor=<rgba 1 0 0.9844 1>}
  NamedRerouteDeclaration(None=FunctionInput() {bUsePreviewValueAsDefault=true, inputName=Material, inputType=FunctionInput_MaterialAttributes, sortPriority=-2}) {name=IN_Material, nodeColor=<rgba 1 0.0938 0 1>}
  n1 = ScalarParameter() {defaultValue=50, group=Flipbook 06 Debug, parameterName=Flipbook.DebugUVSize, sortPriority=1}
  n2 = Multiply(
   A=StaticSwitchParameter(True=TextureCoordinate() {coordinateIndex=1}, False=TextureCoordinate()) {group=Flipbook 06 Debug, parameterName=Flipbook.DebugUV2, sortPriority=0}
   B=@n1
  ) {constB=100}
  n3 = Multiply(A=@n1) {constB=0.5}
  n4 = Subtract(A=ComponentMask(None=@n2) {r=true}, B=@n3)
  n5 = Constant()
  n6 = Add(A=@n3, B=Multiply(A=ComponentMask(None=@n2) {g=true}) {constB=-1}) {constA=100}
  n7 = VectorParameter() {group=Flipbook 06 Debug, parameterName=Flipbook.DebugUVOffset, sortPriority=2, defaultValue=<rgba 0 0 0 1>}
  n8 = Add(
   A=ObjectPositionWS()
   B=Add(
    A=MaterialFunctionCall(X=@n4, Y=@n5, Z=@n6) {materialFunction=&MakeFloat3}.Result
    B=@n7.RGB
   )
  )
  n9 = WorldPosition()
  NamedRerouteDeclaration(None=LinearInterpolate(A=Subtract(A=@n8, B=@n9.XYZ), B=NamedRerouteUsage(), Alpha=NamedRerouteUsage())) {name=DebugUVWPO, nodeColor=<rgba 0 0.4141 1 1>}
  NamedRerouteDeclaration(None=LinearInterpolate(A=NamedRerouteUsage(), B=NamedRerouteUsage(), Alpha=NamedRerouteUsage())) {name=DebugVertColorWPO, nodeColor=<rgba 1 0 0.8203 1>}
  n10 = NamedRerouteUsage()
  NamedRerouteDeclaration(None=LinearInterpolate(
    A=GetMaterialAttributes(None=@n10) {attributeGetTypes [#1]}.BaseColor
    B=NamedRerouteUsage()
    Alpha=ScalarParameter() {defaultValue=1, group=Flipbook 06 Debug, parameterName=Flipbook.DebugVertColor_Blend, sortPriority=24}
   )) {name=DebugVertColor, nodeColor=<rgba 0.3828 0 1 1>}
  n11 = ChannelMaskParameter(None=NamedRerouteUsage()) {group=Flipbook 06 Debug, parameterName=DebugVertColorChannel, sortPriority=21}
  NamedRerouteDeclaration(None=Multiply(A=Step(X=@n11.R) {constY=0.025}, B=Step(Y=@n11.R) {constX=0.975})) {name=DebugVertexMask, nodeColor=<rgba 0.1953 1 0 1>}
  NamedRerouteDeclaration(None=Subtract(A=ActorPositionWS(), B=WorldPosition().XYZ)) {name=VertexCrunch, nodeColor=<rgba 1 0 0.5156 1>}
  n12 = NamedRerouteUsage()
  n13 = NamedRerouteUsage()
  n14 = Subtract(A=Multiply(A=Fmod(A=@n12, B=@n13), B=NamedRerouteUsage()) {constB=50}, B=NamedRerouteUsage())
  n15 = Constant()
  n16 = Divide(A=NamedRerouteUsage(), B=NamedRerouteUsage()) {constB=3}
  n17 = Subtract(A=Multiply(A=NamedRerouteUsage(), B=Floor(None=@n16)), B=NamedRerouteUsage())
  n18 = VectorParameter() {group=Flipbook 06 Debug, parameterName=Flipbook.DebugVertColorOffset, sortPriority=23, defaultValue=<rgba 0 0 0 1>}
  NamedRerouteDeclaration(None=Add(
    A=MaterialFunctionCall(X=@n14, Y=@n15, Z=@n17) {materialFunction=&MakeFloat3}.Result
    B=@n18.RGB
   )) {name=DebugVertexOffset, nodeColor=<rgba 0 0.8359 1 1>}
  NamedRerouteDeclaration(None=ScalarParameter() {defaultValue=20, group=Flipbook 06 Debug, parameterName=Flipbook.DebugVertColor_CellSize, sortPriority=22}) {name=DebugVertColorCellSize, nodeColor=<rgba 1 0.8438 0 1>}
  NamedRerouteDeclaration(None=Constant() {r=3}) {name=DebugVertColorColumnCount, nodeColor=<rgba 1 0.2578 0 1>}
  n19 = Subtract(A=NamedRerouteUsage())
  n20 = NamedRerouteUsage()
  NamedRerouteDeclaration(None=Multiply(A=Multiply(A=@n19, B=@n20)) {constB=0.5}) {name=DebugVertColorOffset, nodeColor=<rgba 0.7109 0 1 1>}
  n21 = Multiply(A=@n11.R) {constB=20}
  NamedRerouteDeclaration(None=Round(None=Subtract(A=@n21) {constB=2})) {name=DebugVertColorIndex, nodeColor=<rgba 0 1 0.3906 1>}
 ]}
