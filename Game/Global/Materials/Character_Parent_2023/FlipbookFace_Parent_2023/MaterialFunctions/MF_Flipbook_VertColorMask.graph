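# Material function for flipbook face character materials that outputs a Vertex Mask selecting the vertices whose color value matches a given index.
# It offsets the scalar Index and scales it into evenly spaced target values, one twentieth apart.
# A SphereMask compares that target to the scalar VertexColor input, using a radius narrower than half a step so neighboring indices stay separate.
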MaterialFunction /Game/Global/Materials/Character_Parent_2023/FlipbookFace_Parent_2023/MaterialFunctions/MF_Flipbook_VertColorMask
graph {nodes [
  n1 = Divide(A=Add(A=FunctionInput() {bUsePreviewValueAsDefault=true, inputName=Index, inputType=FunctionInput_Scalar})) {constB=20}
  FunctionOutput(None=SphereMask(
    A=Add(A=@n1) {constB=0.05}
    B=FunctionInput() {bUsePreviewValueAsDefault=true, inputName=VertexColor, inputType=FunctionInput_Scalar, sortPriority=20}
   ) {attenuationRadius=0.02}) {outputName=Vertex Mask}
 ]}
