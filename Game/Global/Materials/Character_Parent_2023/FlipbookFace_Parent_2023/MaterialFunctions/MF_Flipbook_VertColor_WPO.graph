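# Material function for flipbook character faces that hides vertices by vertex color, outputting a World Position Offset and the hide mask.
# Step tests exclude near-black and near-white vertex colors, then the Vertex Mask input is subtracted and the result rounded and saturated.
# That mask scales an object-position-minus-world-position vector, collapsing masked vertices onto the object position.
# The near-black test's max reads red twice and omits green.
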MaterialFunction /Game/Global/Materials/Character_Parent_2023/FlipbookFace_Parent_2023/MaterialFunctions/MF_Flipbook_VertColor_WPO
graph {nodes [
  FunctionOutput(None=NamedRerouteUsage()) {outputName=Flipbook WPO}
  FunctionOutput(None=NamedRerouteUsage()) {outputName=Vertex Color Hide Mask, sortPriority=1}
  NamedRerouteDeclaration(None=Subtract(A=ObjectPositionWS(), B=WorldPosition().XYZ)) {name=VertexColorHide, nodeColor=<rgba 0.6875 0 1 1>}
  n1 = VertexColor()
  n2 = Max(A=Max(A=@n1.R, B=@n1.R), B=@n1.B)
  n3 = Min(A=Min(A=@n1.R, B=@n1.G), B=@n1.B)
  n4 = OneMinus(None=Add(A=Step(Y=@n2) {constX=0.001}, B=Step(X=@n3) {constX=0.99, constY=0.999}))
  n5 = NamedRerouteUsage()
  n6 = Saturate(None=Round(None=Subtract(A=@n4, B=@n5)))
  NamedRerouteDeclaration(None=Multiply(A=NamedRerouteUsage(), B=NamedRerouteDeclaration(None=@n6) {name=VertexColorHideMask, nodeColor=<rgba 0 0.8359 1 1>})) {name=Flipbook_WPO, nodeColor=<rgba 1 0 0.5391 1>}
  NamedRerouteDeclaration(None=FunctionInput() {inputName=Vertex Mask, inputType=FunctionInput_Scalar}) {name=VertexMask, nodeColor=<rgba 1 0.5859 0 1>}
 ]}
