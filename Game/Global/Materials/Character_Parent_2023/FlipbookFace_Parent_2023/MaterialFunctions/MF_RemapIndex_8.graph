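# Material function for the flipbook face character materials that remaps a scalar index through an eight-entry lookup table, outputting Remapped Index.
# The table arrives as two Vector4 inputs, Input_0-3 and Input_4-7, which default to an identity mapping.
# Floor of Index divided by 4 switches between the vectors, then Index modulo 4 switches among the broken-out components of the selected one.
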MaterialFunction /Game/Global/Materials/Character_Parent_2023/FlipbookFace_Parent_2023/MaterialFunctions/MF_RemapIndex_8
graph {expressions {
  MaterialExpressionConstant_1 {class=MaterialExpressionConstant, properties {r=4}}
  MaterialExpressionDivide_0 {
   class=MaterialExpressionDivide
   inputs [
    A {from=MaterialExpressionNamedRerouteDeclaration_0}
    B {from=""}
   ]
   properties {constB=4}
  }
  MaterialExpressionFloor_0 {class=MaterialExpressionFloor, inputs [None {from=MaterialExpressionDivide_0, output=""}], properties {}}
  MaterialExpressionFmod_1 {
   class=MaterialExpressionFmod
   inputs [
    A {from=MaterialExpressionNamedRerouteUsage_0, output=""}
    B {from=MaterialExpressionConstant_1, output=""}
   ]
   properties {}
  }
  MaterialExpressionFunctionInput_2 {
   class=MaterialExpressionFunctionInput
   inputs [Preview {from="", output=""}]
   properties {bUsePreviewValueAsDefault=true, inputName=Index, inputType=FunctionInput_Scalar}
  }
  MaterialExpressionFunctionInput_3 {
   class=MaterialExpressionFunctionInput
   inputs [Preview {from="", output=""}]
   properties {bUsePreviewValueAsDefault=true, inputName=Input_0-3, inputType=FunctionInput_Vector4, sortPriority=1, previewValue=<0 1 2 3>}
  }
  MaterialExpressionFunctionInput_4 {
   class=MaterialExpressionFunctionInput
   inputs [Preview {from="", output=""}]
   properties {bUsePreviewValueAsDefault=true, inputName=Input_4-7, inputType=FunctionInput_Vector4, sortPriority=2, previewValue=<4 5 6 7>}
  }
  MaterialExpressionFunctionOutput_0 {class=MaterialExpressionFunctionOutput, inputs [None {from=MaterialExpressionSwitch_4, output=""}], properties {outputName=Remapped Index}}
  MaterialExpressionMaterialFunctionCall_1 {
   class=MaterialExpressionMaterialFunctionCall
   inputs [Float4 {from=MaterialExpressionSwitch_0}]
   properties {materialFunction=&BreakOutFloat4Components}
  }
  MaterialExpressionNamedRerouteDeclaration_0 {
   class=MaterialExpressionNamedRerouteDeclaration
   inputs [None {from=MaterialExpressionFunctionInput_2}]
   properties {name=Index, nodeColor=<rgba 0 0.7891 1 1>}
  }
  MaterialExpressionNamedRerouteUsage_0 {class=MaterialExpressionNamedRerouteUsage}
  MaterialExpressionSwitch_0 {
   class=MaterialExpressionSwitch
   inputs [
    SwitchValue {from=MaterialExpressionFloor_0, output=""}
    Default {from="", output=""}
    None {from=MaterialExpressionFunctionInput_3, output=""}
    None {from=MaterialExpressionFunctionInput_4, output=""}
   ]
   properties {inputs [
     {
      inputName=None
      input {
       expression=&self:MaterialExpressionFunctionInput_3
       inputName=None
       mask=0
       maskA=0
       maskB=0
       maskG=0
       maskR=0
       outputIndex=0
      }
     }
     {
      inputName=None
      input {
       expression=&self:MaterialExpressionFunctionInput_4
       inputName=None
       mask=0
       maskA=0
       maskB=0
       maskG=0
       maskR=0
       outputIndex=0
      }
     }
    ]}
  }
  MaterialExpressionSwitch_4 {
   class=MaterialExpressionSwitch
   inputs [
    SwitchValue {from=MaterialExpressionFmod_1, output=""}
    Default {from="", output=""}
    None {from=MaterialExpressionMaterialFunctionCall_1, output=R}
    None {from=MaterialExpressionMaterialFunctionCall_1, output=R}
    None {from=MaterialExpressionMaterialFunctionCall_1, output=R}
    None {from=MaterialExpressionMaterialFunctionCall_1, output=R}
   ]
   properties {inputs [
     {
      inputName=None
      input {
       expression=&self:MaterialExpressionMaterialFunctionCall_1
       inputName=None
       mask=0
       maskA=0
       maskB=0
       maskG=0
       maskR=0
       outputIndex=0
      }
     }
     {
      inputName=None
      input {
       expression=&self:MaterialExpressionMaterialFunctionCall_1
       inputName=None
       mask=0
       maskA=0
       maskB=0
       maskG=0
       maskR=0
       outputIndex=1
      }
     }
     {
      inputName=None
      input {
       expression=&self:MaterialExpressionMaterialFunctionCall_1
       inputName=None
       mask=0
       maskA=0
       maskB=0
       maskG=0
       maskR=0
       outputIndex=2
      }
     }
     {
      inputName=None
      input {
       expression=&self:MaterialExpressionMaterialFunctionCall_1
       inputName=None
       mask=0
       maskA=0
       maskB=0
       maskG=0
       maskR=0
       outputIndex=3
      }
     }
    ]}
  }
 }}
