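# Material function for flipbook faces in the 2023 character parent material, producing a flipbook color, alpha and vertex mask.
# MF_FlipbookTexture samples the texture from index, row and column counts, and UV offset and scale inputs.
# MF_Flipbook_VertColorMask derives the mask from an index and vertex color.
# Static switches select between these results and constants, with the mask switch wired inversely to color and alpha.
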MaterialFunction /Game/Global/Materials/Character_Parent_2023/FlipbookFace_Parent_2023/MaterialFunctions/MF_Sample_Flipbook
graph {nodes [
  FunctionInput() {inputName=----------------------, inputType=FunctionInput_Scalar, sortPriority=-1}
  FunctionInput() {inputName=---------------------, inputType=FunctionInput_Scalar, sortPriority=1}
  FunctionOutput(None=NamedRerouteUsage()) {outputName=Flipbook Color, sortPriority=1}
  FunctionOutput(None=NamedRerouteUsage()) {outputName=Flipbook Alpha, sortPriority=2}
  FunctionOutput(None=NamedRerouteUsage()) {outputName=Flipbook Vertex Mask}
  NamedRerouteDeclaration(None=FunctionInput() {inputName=Vertex Color, inputType=FunctionInput_Scalar}) {name=VertexColor, nodeColor=<rgba 1 0.9844 0 1>}
  n1 = MaterialFunctionCall!(
   Flipbook Index=NamedRerouteUsage()
   Flipbook Texture=NamedRerouteUsage()
   Flipbook Column Count=NamedRerouteUsage()
   Flipbook Row Count=NamedRerouteUsage()
   UVs=NamedRerouteUsage()
   UV Offset=NamedRerouteUsage()
   UV Scale=NamedRerouteUsage()
   MirrorLeftUV
  ) {materialFunction=/Game/Global/Materials/Character_Parent_2023/FlipbookFace_Parent_2023/MaterialFunctions/Common/MF_FlipbookTexture}
  n2 = NamedRerouteUsage()
  NamedRerouteDeclaration(None=StaticSwitch(True=@n1.Alpha, False=Constant(), Value=@n2)) {name=OUT_FlipbookAlpha, nodeColor=<rgba 0.0019 0.0629 1 1>}
  NamedRerouteDeclaration(None=FunctionInput() {inputName=Use Texture, inputType=FunctionInput_StaticBool, sortPriority=-4}) {name=UseTexture, nodeColor=<rgba 0 1 0.3438 1>}
  NamedRerouteDeclaration(None=FunctionInput() {inputName=Curve Index, inputType=FunctionInput_Scalar, sortPriority=-5}) {name=Curves_Index, nodeColor=<rgba 1 0.9844 0 1>}
  NamedRerouteDeclaration(None=FunctionInput() {inputName=Flipbook Texture, inputType=FunctionInput_Texture2D, sortPriority=5}) {name=FlipbookTexture, nodeColor=<rgba 1 0.1875 0 1>}
  NamedRerouteDeclaration(None=FunctionInput() {inputName=Row Count, inputType=FunctionInput_Scalar, sortPriority=7}) {name=RowCount, nodeColor=<rgba 1 0 0.7734 1>}
  NamedRerouteDeclaration(None=FunctionInput() {inputName=Column Count, inputType=FunctionInput_Scalar, sortPriority=6}) {name=ColumnCount, nodeColor=<rgba 1 0 0.7734 1>}
  NamedRerouteDeclaration(None=FunctionInput() {inputName=UVs, inputType=FunctionInput_Vector2, sortPriority=8}) {name=UVs, nodeColor=<rgba 1 0.5859 0 1>}
  NamedRerouteDeclaration(None=FunctionInput() {inputName=UV Scale, inputType=FunctionInput_Vector2, sortPriority=10}) {name=UV Scale, nodeColor=<rgba 0.2656 0 1 1>}
  NamedRerouteDeclaration(None=FunctionInput() {inputName=UV Offset, inputType=FunctionInput_Vector2, sortPriority=9}) {name=UVOffset, nodeColor=<rgba 0.0547 1 0 1>}
  n3 = NamedRerouteUsage()
  n4 = NamedRerouteUsage()
  NamedRerouteDeclaration(None=StaticSwitch(
    True=Constant()
    False=MaterialFunctionCall!(Index=@n3, VertexColor=@n4) {materialFunction=&/MF_Flipbook_VertColorMask}.Vertex Mask
    Value=NamedRerouteUsage()
   )) {name=OUT_VertexMask, nodeColor=<rgba 0.0019 0.0629 1 1>}
  NamedRerouteDeclaration(None=StaticSwitch(True=@n1.Color, False=Constant(), Value=@n2)) {name=OUT_FlipbookColor, nodeColor=<rgba 0.0019 0.0629 1 1>}
 ]}
